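# Control Rig for the AlpacaLean Gel backpack cosmetic that offsets the pack bone and animates its lid opening.
# A constant, full-weight transform nudges the pack bone, then a second transform rotates and slightly shifts the child lid bone.
# The Open_lid variable drives the lid weight through a clamped AlphaInterp, which rises faster than it falls, so the lid opens quicker than it closes.
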ControlRigBlueprint /BRCosmetics/Accessories/FORT_Backpacks/Backpack_M_MED_AlpacaLean_Gel/Meshes/M_MED_AlpacaLean_Gel_Pack_CtrlRig
graphs [
 RigVMFunctionLibrary {links [], nodes []}
 RigVMModel {nodes [
   RigUnit_BeginExecution = &RigUnit_BeginExecution()
   &RigUnit_ModifyTransforms(
    ItemToModify=((Item=(Type=Bone,Name="pack"),Transform=(Rotation=(X=0,Y=0,Z=0,W=1),Translation=(X=0,Y=-1,Z=0),Scale3D=(X=1,Y=1,Z=1))))
    Weight=1
    WeightMinimum=0
    ExecutePin=@RigUnit_BeginExecution.ExecutePin
   )
   AlphaInterp = &RigVMFunction_AlphaInterp(InterpSpeedDecreasing=10, InterpSpeedIncreasing=20, Scale=1, bClampResult=true, Value=RigVMVariableNode() {variable=Open_lid}.Value)
   > &RigUnit_ModifyTransforms(
    ItemToModify=((Item=(Type=Bone,Name="lid"),Transform=(Rotation=(X=0,Y=0.8192,Z=0,W=0.5736),Translation=(X=0,Y=0,Z=-0.3),Scale3D=(X=1,Y=1,Z=1))))
    WeightMinimum=0
    Weight=@AlphaInterp.Result
   )
  ]}
]
hierarchy [
 root {loc=<0 0 0>, rot=<0 0 0 1>}
 pack {parent=root, loc=<0 -0.7316 0>, rot=<0 0 0 1>}
 lid {parent=pack, loc=<8.5203 -0.8665 -2.8898>, rot=<0 -0.7071 -0.7071 0>}
]
variables [Open_lid {type=double, default=0}]
