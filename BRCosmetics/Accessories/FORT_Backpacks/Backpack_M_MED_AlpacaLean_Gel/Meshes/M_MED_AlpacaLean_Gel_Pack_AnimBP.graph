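# Animation Blueprint for the AlpacaLean Gel backpack, adding body-following motion and a lid-opening flourish.
# Each update maps spine_05's delta from ref pose, relative to spine_01, to additive pack bone offsets.
# A Control Rig receives Open_lid, which chained delays animate as a bouncy open, hold and close.
# The lid sequence plays on enemy-elimination anim events, and in the front end shortly after start, then periodically.
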AnimBlueprint /BRCosmetics/Accessories/FORT_Backpacks/Backpack_M_MED_AlpacaLean_Gel/Meshes/M_MED_AlpacaLean_Gel_Pack_AnimBP
target_skeleton=&/M_MED_AlpacaLean_Gel_Pack_Skeleton
graphs [
 AnimGraph {nodes [
   n1 = Control Rig!(Source=Input Pose(), Alpha, Open_lid:double=Get Open_lid() -> (Open_lid:double)) -> (Pose) {controlRigAssetReference {blueprintRigClass=&/M_MED_AlpacaLean_Gel_Pack_CtrlRig.M_MED_AlpacaLean_Gel_Pack_CtrlRig_C}}
   n2 = "Transform (Modify) Bone - Bone: pack"(
    ComponentPose=Local To Component(LocalPose=@n1)
    Translation=Get Backpack Translation() -> (Backpack Translation:Vector)
    Rotation=Get Backpack Rotate() -> (Backpack Rotate:Rotator)
   ) {rotationMode=BMM_Additive, translationMode=BMM_Additive, alphaBoolBlend {blendOption=Linear}, boneToModify {boneName=pack}}
   Output Pose(Result=Component To Local(ComponentPose=@n2)) {layerGroup=None}
  ]}
 EventGraph {nodes [
   Event BlueprintUpdateAnimation()
   n1 = BreakTransform(InTransform=GetDeltaTransformFromRefPose(self=GetOwnerSkeletalMeshComponent(), BoneName=spine_05, BaseName=spine_01))
   n2 = BreakRotator(InRot=@n1.Rotation)
   n3 = MapRangeClamped(Value=@n2.Yaw, InRangeA=-5, InRangeB=-60, OutRangeB=55)
   n4 = MapRangeClamped(Value=BreakVector(InVec=@n1).Y, InRangeA=0, InRangeB=12, OutRangeB=10)
   > Set Backpack Rotate(Backpack Rotate:Rotator=MakeRotator(Roll=float + float(A=@n4, B=@n3), Pitch=float / float(A=@n3, B=5), Yaw=MapRangeClamped(Value=@n2, InRangeA=0, InRangeB=150, OutRangeB=30))) -> (Output_Get:Rotator)
   n5 = MapRangeClamped(Value=@n2, InRangeA=0, InRangeB=150, OutRangeB=-3)
   n6 = MapRangeClamped(Value=@n2.Yaw, InRangeA=-20, InRangeB=-60, OutRangeB=-10)
   > Set Backpack Translation(Backpack Translation:Vector=MakeVector(
     X=MapRangeClamped(Value=@n2, InRangeA=0, InRangeB=130, OutRangeB=3)
     Y=float + float(A=@n6, B=@n5)
     Z=MapRangeClamped(Value=@n2.Yaw, InRangeA=-20, InRangeB=-60)
    )) -> (Output_Get:Vector)
   > Branch(Condition=Get bIsInFrontEnd())
   > n7 = Branch(Condition="integer > integer"(A=Get Count() -> (Count:int=0), B=1))
   > n8 = Delay(Duration=10)
   Set Open_lid(execute=[@n9.then, @n8.then], Open_lid:double=1) -> (Output_Get:double)
   > Delay(Duration=0.1)
   > Set Open_lid(Open_lid=0.9)
   > Delay(Duration=0.1)
   > Set Open_lid(Open_lid=1)
   > Delay(Duration=0.08)
   > Set Open_lid(Open_lid=0.97)
   > Delay(Duration=0.05)
   > Set Open_lid(Open_lid=1)
   > Delay(Duration=2)
   > Set Open_lid(Open_lid=0)
   Set Count(execute=@n7.else, Count:int=int + int(A=Get Count() -> (Count=0))) -> (Output_Get:int=0)
   > n9 = Delay(Duration=1)
   n10 = Event BlueprintProcessAnimInputEvent()
   > Branch(Condition=IsNamed(self=@n10.AnimInputEvent, InEventName=aie.EnemyPlayerEliminated))
   > Delay(Duration=0.2)
   > Set Open_lid(Open_lid=1)
   > Delay(Duration=0.1)
   > Set Open_lid(Open_lid=0.9)
   > Delay(Duration=0.1)
   > Set Open_lid(Open_lid=1)
   > Delay(Duration=0.08)
   > Set Open_lid(Open_lid=0.97)
   > Delay(Duration=0.05)
   > Set Open_lid(Open_lid=1)
   > Delay(Duration=2)
   > Set Open_lid(Open_lid=0)
   TryGetPawnOwner()
  ]}
]
variables [
 Backpack Rotate {type=Rotator}
 Backpack Translation {type=Vector}
 Count {type=integer}
 Open_lid {type=number}
]
